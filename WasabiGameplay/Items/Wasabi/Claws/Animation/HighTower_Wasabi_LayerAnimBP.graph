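# Item animation layer Blueprint for the HighTower Wasabi claws on the Fortnite player skeleton, supplying melee guard poses.
# ItemUpperBody runs a Block state machine from cached input through intro, guard loop and exit, keyed to bIsMeleeGuarding.
# That flag also switches jump, fall, jetpack and zipline additives between assault-rifle (guarding) and no-weapon sets.
# ItemSkeletalControl adds RootRotationOffset to Root, weighted by the EnableAccelYawCorrection curve; most other layers are unwired.
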AnimBlueprint /WasabiGameplay/Items/Wasabi/Claws/Animation/HighTower_Wasabi_LayerAnimBP
target_skeleton=&Fortnite_M_Avg_Player_Skeleton
graphs [
 AnimGraph {nodes [Output Pose() {layerGroup=None}]}
 EventGraph {links [], nodes []}
 ItemAimOffset {nodes [
   Input Pose!() -> (Pose, AimOffsetAlpha:double, Yaw:double, Pitch:double) {name=InPose_AimOffset}
   Input Pose!() -> (Pose) {name=InPose_PreAimOffsets}
   Output Pose() {layerGroup=Item Override}
  ]}
 ItemFallAdditive {nodes [
   Input Pose() {name=InPoseItemFallAdditive}
   n1 = Get FallPlayRate() -> (FallPlayRate:float)
   Output Pose(Result=Blend Poses by bool(
     BlendPose_0="Sequence Player (additive) 'M_JumpFall_ARDS_Additive'"!(PlayRate:float=@n1) -> (Pose) {bLoopAnimation=false, sequence=/Game/Animation/Game/MainPlayer/Locomotion/Medium/Male/Jump/AssaultRifle/M_JumpFall_ARDS_Additive}
     BlendPose_1="Sequence Player (additive) 'JumpFall_NoWep_CMM_Additive'"!(PlayRate:float=@n1) -> (Pose) {bLoopAnimation=false, sequence=/Game/Animation/Game/MainPlayer/Locomotion/Medium/Male/Jump/NoWep/JumpFall_NoWep_CMM_Additive}
     bActiveValue=Get bIsMeleeGuarding() -> (bIsMeleeGuarding:bool)
    ) {blendType=Linear}) {layerGroup=Item Override}
  ]}
 ItemFallLandAdditive {nodes [
   Input Pose() {name=InPoseItemFallLandAdditive}
   Output Pose() {layerGroup=Item Override}
  ]}
 ItemFallLevel2Override {nodes [
   Input Pose() {name=InPose_FallLevel2}
   Output Pose() {layerGroup=Item Override}
  ]}
 ItemFinalPoseOverride {nodes [
   Input Pose() {name=InFinalPose}
   Output Pose() {layerGroup=Item Override}
  ]}
 ItemFullBodyAimOffsets {nodes [
   Input Pose!() -> (Pose, FullBody_AimOffsetAlpha:double, Aim_Yaw:double, Aim_Pitch:double) {name=InPoseFullBodyAimOffsets}
   Input Pose!() -> (Pose) {name=InPosePreFullBodyAimOffsets}
   Output Pose() {layerGroup=Item Override}
  ]}
 ItemFullBodyOverride {nodes [
   Input Pose() {name=InPoseFullBody}
   Output Pose() {layerGroup=Item Override}
  ]}
 ItemIdleOverride {nodes [
   Input Pose() {name=InPose_Idle}
   Output Pose() {layerGroup=Item Override}
  ]}
 ItemJetPackJumpAdditive {nodes [Output Pose(Result=Blend Poses by bool(
     BlendPose_0="Sequence Player (additive) 'JetpackApex_ARDS_Loop_Additive'"() {sequence=/Game/Animation/Game/MainPlayer/Locomotion/Medium/Male/Jump/AssaultRifle/JetpackApex_ARDS_Loop_Additive}
     BlendPose_1="Sequence Player (additive) 'JetpackApex_NoWep__Loop_CMM_Additive'"() {sequence=/Game/Animation/Game/MainPlayer/Locomotion/Medium/Male/Jump/NoWep/JetpackApex_NoWep__Loop_CMM_Additive}
     BlendTime_1=1
     bActiveValue=Get bIsMeleeGuarding()
    ) {blendType=Linear}) {layerGroup=Item Override}]}
 ItemJetPackStartAdditive {nodes [Output Pose() {layerGroup=Item Override}]}
 ItemJumpLoopAdditive {nodes [
   Input Pose() {name=InPoseItemJumpLoopAdditive}
   Output Pose(Result=Blend Poses by bool(
     BlendPose_0="Sequence Player (additive) 'M_JumpApex_ARDS_Loop_Additive'"() {playRate=0.5, sequence=/Game/Animation/Game/MainPlayer/Locomotion/Medium/Male/Jump/AssaultRifle/M_JumpApex_ARDS_Loop_Additive}
     BlendPose_1="Sequence Player (additive) 'JumpApex_NoWep__Loop_CMM_Additive'"() {sequence=/Game/Animation/Game/MainPlayer/Locomotion/Medium/Male/Jump/NoWep/JumpApex_NoWep__Loop_CMM_Additive}
     bActiveValue=Get bIsMeleeGuarding()
    ) {blendType=Linear}) {layerGroup=Item Override}
  ]}
 ItemJumpOverride {nodes [
   Input Pose() {name=InPoseJump}
   Output Pose() {layerGroup=Item Override}
  ]}
 ItemJumpUpAdditive {nodes [
   Input Pose() {name=InPoseItemJumpUpAdditive}
   Output Pose(Result=Blend Poses by bool(
     BlendPose_0="Sequence Player (additive) 'M_JumpUp_ARDS_Additive'"() {bLoopAnimation=false, sequence=/Game/Animation/Game/MainPlayer/Locomotion/Medium/Male/Jump/AssaultRifle/M_JumpUp_ARDS_Additive}
     BlendPose_1="Sequence Player (additive) 'JumpUp_NoWep_CMM_Additive'"() {bLoopAnimation=false, sequence=/Game/Animation/Game/MainPlayer/Locomotion/Medium/Male/Jump/NoWep/JumpUp_NoWep_CMM_Additive}
     bActiveValue=Get bIsMeleeGuarding()
    ) {blendType=Linear}) {layerGroup=Item Override}
  ]}
 ItemPreAimOffsets {nodes [
   Input Pose() {name=InPosePreAimOffsets}
   Output Pose() {layerGroup=Item Override}
  ]}
 ItemPreIK {nodes [
   Input Pose() {name=InPose_PreIK}
   Output Pose() {layerGroup=Item Override}
  ]}
 ItemRightHandIK {nodes [
   Input Pose() {name=InPosePreRightHandIK}
   Input Pose() {name=InPosePostRightHandIK}
   Output Pose() {layerGroup=Item Override}
  ]}
 ItemSkeletalControl {nodes [
   n1 = Local To Component(LocalPose=Input Pose())
   n2 = Get RootRotationOffset() -> (RootRotationOffset:Rotator)
   Output Pose(Result=Component To Local(ComponentPose="Transform (Modify) Bone - Bone: Root"!(ComponentPose=@n1, AlphaCurveName:Name=EnableAccelYawCorrection, Rotation=@n2) -> (Pose) {alphaInputType=Curve, rotationMode=BMM_Additive, alphaBoolBlend {blendOption=Linear}, boneToModify {boneName=Root}})) {layerGroup=Item Override}
  ]}
 ItemSwimJumpFallLoop {nodes [Output Pose() {layerGroup=Item Override}]}
 ItemSwimJumpSurfaceEnd {nodes [Output Pose() {layerGroup=Item Override}]}
 ItemSwimJumpSurfaceLoop {nodes [Output Pose() {layerGroup=Item Override}]}
 ItemUpperBody {nodes [
   AnimGraphNode_StateMachine_0 = Block() {is=title AnimGraphNode_StateMachine}
   Output Pose(Result=@AnimGraphNode_StateMachine_0) {layerGroup=Item Override}
   Save cached pose 'UpperBody_CachePose'(Pose=Input Pose!() -> (Pose, InputParam:Fort Anim Input Adjusted Aim Structure) {name=InPoseUpperBody}) {cacheName=UpperBody_CachePose}
  ]}
 Block {
  entry=Default
  path=ItemUpperBody.AnimGraphNode_StateMachine_0.Block
  ids=id
  nodes [
   AnimStateNode_0 = GuardLoop() {is=title AnimStateNode, stateEntered {guid=#1, notifyName=GuardLoopEntered}}
   AnimStateNode_11 = Default() {is=title AnimStateNode}
   AnimStateNode_12 = GuardToIdle() {is=title AnimStateNode}
   AnimStateNode_5 = GuardIntro() {is=title AnimStateNode}
   AnimStateTransitionNode_13 = AnimStateTransitionNode() {crossfadeDuration=0.1}
   AnimStateTransitionNode_14 = AnimStateTransitionNode()
   AnimStateTransitionNode_15 = AnimStateTransitionNode() {bAutomaticRuleBasedOnSequencePlayerInState=true}
   AnimStateTransitionNode_5 = AnimStateTransitionNode() {bAutomaticRuleBasedOnSequencePlayerInState=true, crossfadeDuration=0.1}
  ]
  links [
   {
    dst=GuardIntro
    rule=ItemUpperBody.AnimGraphNode_StateMachine_0.Block.AnimStateTransitionNode_13.Transition
    src=Default
    via=AnimStateTransitionNode_13
   }
   {
    dst=GuardLoop
    rule=ItemUpperBody.AnimGraphNode_StateMachine_0.Block.AnimStateTransitionNode_5.Transition
    src=GuardIntro
    via=AnimStateTransitionNode_5
   }
   {
    dst=GuardToIdle
    rule=ItemUpperBody.AnimGraphNode_StateMachine_0.Block.AnimStateTransitionNode_14.Transition
    src=GuardLoop
    via=AnimStateTransitionNode_14
   }
   {
    dst=Default
    rule=ItemUpperBody.AnimGraphNode_StateMachine_0.Block.AnimStateTransitionNode_15.Transition
    src=GuardToIdle
    via=AnimStateTransitionNode_15
   }
  ]
 }
 GuardLoop {
  path=ItemUpperBody.AnimGraphNode_StateMachine_0.Block.AnimStateNode_0.GuardLoop
  nodes [Output Animation Pose(Result=Sequence Player 'HighTower_Wasabi_Block_Pose'() {sequence=&/HighTower_Wasabi_Block_Pose})]
 }
 Default {
  path=ItemUpperBody.AnimGraphNode_StateMachine_0.Block.AnimStateNode_11.Default
  nodes [Output Animation Pose(Result=Use cached pose 'UpperBody_CachePose'())]
 }
 GuardToIdle {
  path=ItemUpperBody.AnimGraphNode_StateMachine_0.Block.AnimStateNode_12.GuardToIdle
  nodes [Output Animation Pose(Result=Sequence Player 'HighTower_Wasabi_Block_Exit'() {sequence=&/HighTower_Wasabi_Block_Exit})]
 }
 GuardIntro {
  path=ItemUpperBody.AnimGraphNode_StateMachine_0.Block.AnimStateNode_5.GuardIntro
  nodes [Output Animation Pose(Result=Sequence Player 'HighTower_Wasabi_Block_Intro'() {sequence=&/HighTower_Wasabi_Block_Intro})]
 }
 Transition {
  path=ItemUpperBody.AnimGraphNode_StateMachine_0.Block.AnimStateTransitionNode_13.Transition
  nodes [
   Result()
   Reroute Node(InputPin:bool=Get bIsMeleeGuarding()) -> (OutputPin:bool)
  ]
 }
 Transition {
  path=ItemUpperBody.AnimGraphNode_StateMachine_0.Block.AnimStateTransitionNode_14.Transition
  nodes [Result(bCanEnterTransition=NOT Boolean(A=Get bIsMeleeGuarding()))]
 }
 Transition {path=ItemUpperBody.AnimGraphNode_StateMachine_0.Block.AnimStateTransitionNode_15.Transition, nodes [Result()]}
 Transition {path=ItemUpperBody.AnimGraphNode_StateMachine_0.Block.AnimStateTransitionNode_5.Transition, nodes [Result()]}
 ItemZipLineStartAdditive {nodes [Output Pose(Result=Blend Poses by bool(
     BlendPose_0="Sequence Player (additive) 'ZipLine_Entry_ARDS_Additive'"() {bLoopAnimation=false, sequence=/Game/Animation/Game/MainPlayer/Combat/Gadgets/Medium/Male/ZipLine/ZipLine_Entry_ARDS_Additive}
     BlendPose_1="Sequence Player (additive) 'ZipLine_Entry_NoWep_Additive'"() {bLoopAnimation=false, sequence=/Game/Animation/Game/MainPlayer/Combat/Gadgets/Medium/Male/ZipLine/ZipLine_Entry_NoWep_Additive}
     bActiveValue=Get bIsMeleeGuarding()
    ) {blendType=Linear}) {layerGroup=Item Override}]}
]
variables []
