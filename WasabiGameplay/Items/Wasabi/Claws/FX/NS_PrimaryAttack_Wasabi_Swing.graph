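# Niagara ribbon effect for the swing of Wasabi's claws primary attack, drawing short-lived, blue, fading trails.
# Arrow draws one wide ribbon, while Blades_Trails (thin, with a bright flash) and Blades_Trails_Perpedicular each use three ribbon renderers.
# Both blade emitters write two vertically offset positions, apparently for parallel claw trails.
# All emitters take velocity from the "Ribbon camera face" user vector and taper width over lifetime.
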NiagaraSystem /WasabiGameplay/Items/Wasabi/Claws/FX/NS_PrimaryAttack_Wasabi_Swing
dependencies {
 usedDataInterfaces [&NiagaraDataInterfaceCurve, &NiagaraDataInterfaceColorCurve]
 usedDynamicInputs []
 usedModules [&SystemState]
}
emitters {
 Arrow {
  emitterUpdateScript [
   &EmitterState()
   &SpawnRate(SpawnRate=50, Use Spawn Probability=false)
  ]
  particleSpawnScript [
   &InitializeRibbon(
    Lifetime=0.12
    Module.Write Color=true
    Module.Write Lifetime=true
    Module.Write Mass=true
    Module.Write Material Random=true
    Module.Write Position=false
    Module.Write RibbonTwist=true
    Module.Write RibbonWidth=true
    Ribbon Twist:NiagaraFloat=&UniformRangedFloat(Minimum=0, Maximum=0.5)
    Ribbon Width=85
   )
   &AddVelocity(Velocity=$User.Ribbon camera face)
  ]
  particleUpdateScript [
   &ParticleState()
   &SolveForcesAndVelocity(Clamp Velocity=false, Limit Acceleration=false)
   &ScaleRibbonWidth(Ribbon Width Scale=&FloatFromCurve(FloatCurve=DataInterface() {propertyValues {Curve {keys [
         {arriveTangent=-1.2579, interpMode=RCIM_Cubic, time=0, value=1}
         {arriveTangent=-0.9664, interpMode=RCIM_Cubic, leaveTangent=-0.9664, tangentMode=RCTM_User, time=1, value=0}
        ]}}}))
   &Color(
    Color=&ColorFromCurve(ColorCurve=DataInterface() {propertyValues {
       AlphaCurve {keys [<at 0 1 RCIM_Cubic>,<at 1 0 RCIM_Cubic>]}
       BlueCurve {keys [<at 0 1 RCIM_Cubic>,<at 1 1 RCIM_Cubic>]}
       GreenCurve {keys [<at 0 0.2775 RCIM_Cubic>,<at 1 0.2789 RCIM_Cubic>]}
       RedCurve {keys [<at 0 0.1024 RCIM_Cubic>,<at 1 0.1022 RCIM_Cubic>]}
      }})
    Scale Alpha=1.5
   ) {module=Color001}
  ]
  rendererClasses [&NiagaraRibbonRendererProperties]
  renderers [{rendererClass=&NiagaraRibbonRendererProperties}]
 }
 Blades_Trails {
  emitterUpdateScript [
   &EmitterState()
   &SpawnRate(SpawnRate=50, Use Spawn Probability=false)
  ]
  particleSpawnScript [
   &InitializeRibbon(
    Lifetime=0.17
    Module.Write Color=true
    Module.Write Lifetime=true
    Module.Write Mass=true
    Module.Write Material Random=true
    Module.Write Position=false
    Module.Write RibbonTwist=true
    Module.Write RibbonWidth=true
    Ribbon Twist:NiagaraFloat=&UniformRangedFloat(Minimum=0, Maximum=0.5)
    Ribbon Width=12
   )
   /SetVariables(Particles.PositionOffset1:Vector3f=&Add_Vector(A=Linked() {linkedVariable {name=Particles.Position, type=NiagaraPosition}}, B=<0 0 -12>)) {isSetParameters=true, module=SetVariables_#1}
   /SetVariables(Particles.PositionOffset2:Vector3f=&Add_Vector(A=Linked() {linkedVariable {name=Particles.Position, type=NiagaraPosition}}, B=<0 0 12>)) {isSetParameters=true, module=SetVariables_#2}
   &AddVelocity(Velocity=$User.Ribbon camera face)
  ]
  particleUpdateScript [
   &ParticleState()
   &SolveForcesAndVelocity(Clamp Velocity=false, Limit Acceleration=false)
   &ScaleRibbonWidth(Ribbon Width Scale=&FloatFromCurve(FloatCurve=DataInterface() {propertyValues {Curve {keys [
         {arriveTangent=-1.2579, interpMode=RCIM_Cubic, time=0, value=1}
         {arriveTangent=-0.9664, interpMode=RCIM_Cubic, leaveTangent=-0.9664, tangentMode=RCTM_User, time=1, value=0}
        ]}}}))
   &Color(
    Color=&ColorFromCurve(ColorCurve=DataInterface() {propertyValues {
       AlphaCurve {keys [<at 0.1755 0>, {arriveTangent=0.8391, interpMode=RCIM_Cubic, time=0.2218, value=0.3}, <at 0.3254 0.1>, <at 1.0518 0 RCIM_Cubic>]}
       BlueCurve {keys [<at 0.2234 3 RCIM_Cubic>,<at 0.3386 5>,<at 0.4944 0.2812 RCIM_Cubic>]}
       GreenCurve {keys [<at 0.2234 1.2271 RCIM_Cubic>,<at 0.3386 3.8784>,<at 0.4944 0.1826 RCIM_Cubic>]}
       RedCurve {keys [<at 0.2234 0.5986 RCIM_Cubic>,<at 0.3386 3.1909>,<at 0.4944 0.1514 RCIM_Cubic>]}
      }})
    Scale Alpha=1.5
   ) {module=Color001}
  ]
  rendererClasses [&NiagaraRibbonRendererProperties]
  renderers [
   {rendererClass=&NiagaraRibbonRendererProperties}
   {rendererClass=&NiagaraRibbonRendererProperties, rendererIndex=1}
   {rendererClass=&NiagaraRibbonRendererProperties, rendererIndex=2}
  ]
 }
 Blades_Trails_Perpedicular {
  emitterUpdateScript [
   &EmitterState()
   &SpawnRate(SpawnRate=50, Use Spawn Probability=false)
  ]
  particleSpawnScript [
   &InitializeRibbon(
    Lifetime=0.21
    Module.Write Color=true
    Module.Write Lifetime=true
    Module.Write Mass=true
    Module.Write Material Random=true
    Module.Write Position=false
    Module.Write RibbonTwist=true
    Module.Write RibbonWidth=true
    Ribbon Twist:NiagaraFloat=&UniformRangedFloat(Minimum=0, Maximum=0.5)
    Ribbon Width=35
   )
   /SetVariables(Particles.PositionOffset1:Vector3f=&Add_Vector(A=Linked() {linkedVariable {name=Particles.Position, type=NiagaraPosition}}, B=<0 0 -12>)) {isSetParameters=true, module=SetVariables_#1}
   /SetVariables(Particles.PositionOffset2:Vector3f=&Add_Vector(A=Linked() {linkedVariable {name=Particles.Position, type=NiagaraPosition}}, B=<0 0 12>)) {isSetParameters=true, module=SetVariables_#2}
   &AddVelocity(Velocity=$User.Ribbon camera face)
  ]
  particleUpdateScript [
   &ParticleState()
   /SetVariables(Particles.SpriteFacing=HlslExpression() {hlslExpression="dot(Particles.RibbonFacing,  float3(0,0,1))"}) {isSetParameters=true}
   &SolveForcesAndVelocity(Clamp Velocity=false, Limit Acceleration=false)
   &ScaleRibbonWidth(Ribbon Width Scale=&FloatFromCurve(FloatCurve=DataInterface() {propertyValues {Curve {keys [
         {arriveTangent=-1.2579, interpMode=RCIM_Cubic, time=0, value=1}
         {arriveTangent=-0.9664, interpMode=RCIM_Cubic, leaveTangent=-0.9664, tangentMode=RCTM_User, time=1, value=0}
        ]}}}))
   &Color(
    Color=&ColorFromCurve(ColorCurve=DataInterface() {propertyValues {
       AlphaCurve {keys [
         <at 0.1755 0>
         {arriveTangent=0.1729, interpMode=RCIM_Cubic, leaveTangent=0.1729, time=0.2394, value=0.1}
         <at 0.7539 0.1>
         <at 0.9226 0.5>
         <at 1.0518 0 RCIM_Cubic>
        ]}
       BlueCurve {keys [<at 0.2234 1 RCIM_Cubic>]}
       GreenCurve {keys [<at 0.2234 0.3476 RCIM_Cubic>]}
       RedCurve {keys [<at 0.2234 0.1163 RCIM_Cubic>]}
      }})
    Scale Alpha=1.5
   ) {module=Color001}
  ]
  rendererClasses [&NiagaraRibbonRendererProperties]
  renderers [
   {rendererClass=&NiagaraRibbonRendererProperties}
   {rendererClass=&NiagaraRibbonRendererProperties, rendererIndex=1}
   {rendererClass=&NiagaraRibbonRendererProperties, rendererIndex=2}
  ]
 }
}
summary {
 systemName=NS_PrimaryAttack_Wasabi_Swing
 emitters [
  {emitterName=Blades_Trails, rendererClasses [&NiagaraRibbonRendererProperties]}
  {emitterName=Arrow, rendererClasses [&NiagaraRibbonRendererProperties]}
  {emitterName=Blades_Trails_Perpedicular, rendererClasses [&NiagaraRibbonRendererProperties]}
 ]
 userVariables [
  User.Ribbon camera face {type=Vector3f, description="", defaultValue {struct=Vector3f, value=<-0.1 0 0.1>}}
  User.RibbonIDOffset {type=NiagaraInt32}
 ]
}
